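; BP_AISC_Cosmetics_DangerGrape is a Blueprint that supplies cosmetics for the DangerGrape AI in Athena.
; It implements three getters for cosmetic loadout, custom character parts and a deprecated pawn loadout.
; Each getter directly returns the matching player value from the Default category, with no selection or randomization logic of its own.
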(fn GetCosmeticLoadout
  (return (Variables|Default|GetPlayerCosmeticLoadout)))

(fn GetCustomCharacterParts
  (return (Variables|Default|GetPlayerCustomCharacterParts)))

(fn GetLoadout(Deprecated)
  (return (Variables|Default|GetPlayerPawnLoadout_Deprecated)))
